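# Physics asset for the ForestBath backpack cosmetic, defining secondary motion for its ears, wings, necklace and zip.
# All bodies are capsules: ear, wing and zip bases are angularly locked to the pack and carry tightly limited dyn_ bones, while the necklace chain is limited throughout.
# Spine, clavicle and upper-arm capsules cover the wearer's upper body, with asymmetric constraints: free rotation on the left, locked or limited on the right.
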PhysicsAsset /BRCosmetics/Accessories/FORT_Backpacks/Backpack_F_MED_ForestBath/Meshes/F_MED_ForestBath_Pack_PhysicsAsset
bodies {
 base_l_ear [{shapeType=Capsule, length=1.005, radius=2.505, shapeName=base_l_ear_capsule, rotation=<rot 0 -90 0>}]
 base_l_wing_00 [{shapeType=Capsule, length=1.005, radius=2.505, shapeName=base_l_wing_00_capsule, rotation=<rot 0 -90 0>}]
 base_necklace [{shapeType=Capsule, length=1.5, radius=1.005, shapeName=base_necklace_capsule, rotation=<rot 0 -90 0>}]
 base_r_ear [{shapeType=Capsule, length=1.005, radius=2.505, shapeName=base_r_ear_capsule, rotation=<rot 0 -90 0>}]
 base_r_wing_00 [{shapeType=Capsule, length=1.005, radius=2.505, shapeName=base_r_wing_00_capsule, rotation=<rot 0 -90 0>}]
 base_zip [{shapeType=Capsule, length=1.005, radius=2.505, shapeName=base_zip_capsule, rotation=<rot 0 -90 0>}]
 clavicle_l [{
   shapeType=Capsule
   length=4.5259
   radius=7.548
   shapeName=clavicle_r_capsule
   center=<4.962 -1.5855 -0.233>
   rotation=<rot -87.8875 179.9668 0.0055>
  }]
 clavicle_r [{
   shapeType=Capsule
   length=4.5259
   radius=7.548
   shapeName=clavicle_r_capsule
   center=<-4.962 1.5855 0.233>
   rotation=<rot -87.8875 179.9668 -179.9945>
  }]
 dyn_l_ear [{
   shapeType=Capsule
   length=4.3873
   radius=2.1221
   shapeName=dyn_l_ear_capsule
   center=<1.9904 0.2272 -0.4035>
   rotation=<rot 4.4648 -90 3.5053>
  }]
 dyn_l_wing_00 [{
   shapeType=Capsule
   length=2.8633
   radius=2.2322
   shapeName=dyn_l_wing_00_capsule
   center=<-4.4171 -0.5213 0.1641>
   rotation=<rot -31.8238 -90 -15.1649>
  }]
 dyn_l_wing_01 [{
   shapeType=Capsule
   length=6.6286
   radius=1.7056
   shapeName=dyn_l_wing_01_capsule
   center=<-4.4921 -1.5216 -3.3911>
   rotation=<rot 23.4918 -90 38.4798>
  }]
 dyn_necklace_00 [{
   shapeType=Capsule
   length=1.5
   radius=1.4543
   shapeName=dyn_necklace_00_capsule
   center=<0.4185 -0.1745 -0.0056>
   rotation=<rot 0 89.2616 -82.6912>
  }]
 dyn_necklace_01 [{
   shapeType=Capsule
   length=3.01
   radius=1
   shapeName=dyn_necklace_01_capsule
   center=<5.4015 -0.0008 -0.4374>
   rotation=<rot 3.5533 -178.9779 -89.9972>
  }]
 dyn_r_ear [{
   shapeType=Capsule
   length=4.3873
   radius=2.1221
   shapeName=dyn_r_ear_capsule
   center=<-1.9904 -0.2272 0.4035>
   rotation=<rot 4.4648 -90 3.5052>
  }]
 dyn_r_wing_00 [{
   shapeType=Capsule
   length=2.8633
   radius=2.2322
   shapeName=dyn_l_wing_00_capsule
   center=<4.4171 0.5213 -0.1641>
   rotation=<rot -31.8238 -90 164.8351>
  }]
 dyn_r_wing_01 [{
   shapeType=Capsule
   length=6.6286
   radius=1.7056
   shapeName=dyn_l_wing_01_capsule
   center=<4.4921 1.5216 3.3911>
   rotation=<rot 23.4918 -90 -141.5202>
  }]
 dyn_zip [{shapeType=Capsule, length=2.4782, radius=0.5981, shapeName=dyn_zip_capsule, center=<0.028 0.0729 1.3912>, rotation=<rot 0.755 -90 0.9398>}]
 pack [{shapeType=Capsule, length=5, radius=7.426, shapeName=pack_capsule, center=<0.0246 -11.4733 -9.6884>, rotation=<rot 3.5615 -90.4096 -4.7032>}]
 spine_05 [{shapeType=Capsule, length=1.005, radius=3.505, shapeName=spine_05_capsule, rotation=<rot 0 -90 0>}]
 upperarm_l [{
   shapeType=Capsule
   length=9.6921
   radius=5.6066
   shapeName=upperarm_r_capsule
   center=<10.0393 0.005 -0.1361>
   rotation=<rot -86.5346 0.5522 179.4752>
  }]
 upperarm_r [{
   shapeType=Capsule
   length=9.6921
   radius=5.6066
   shapeName=upperarm_r_capsule
   center=<-10.0393 -0.005 0.1361>
   rotation=<rot -86.5346 0.5522 -0.5248>
  }]
}
constraints [
 {
  bone1Name=base_l_ear
  bone2Name=pack
  swing1LimitDegrees=45
  swing1Motion=Locked
  swing2LimitDegrees=45
  swing2Motion=Locked
  twistLimitDegrees=45
  twistMotion=Locked
 }
 {
  bone1Name=base_l_wing_00
  bone2Name=pack
  swing1LimitDegrees=45
  swing1Motion=Locked
  swing2LimitDegrees=45
  swing2Motion=Locked
  twistLimitDegrees=45
  twistMotion=Locked
 }
 {bone1Name=base_necklace, bone2Name=pack, swing1LimitDegrees=5, swing2LimitDegrees=25, twistLimitDegrees=5}
 {
  bone1Name=base_r_ear
  bone2Name=pack
  swing1LimitDegrees=45
  swing1Motion=Locked
  swing2LimitDegrees=45
  swing2Motion=Locked
  twistLimitDegrees=45
  twistMotion=Locked
 }
 {
  bone1Name=base_r_wing_00
  bone2Name=pack
  swing1LimitDegrees=45
  swing1Motion=Locked
  swing2LimitDegrees=45
  swing2Motion=Locked
  twistLimitDegrees=45
  twistMotion=Locked
 }
 {
  bone1Name=base_zip
  bone2Name=pack
  swing1LimitDegrees=45
  swing1Motion=Locked
  swing2LimitDegrees=45
  swing2Motion=Locked
  twistLimitDegrees=45
  twistMotion=Locked
 }
 {bone1Name=dyn_necklace_00, bone2Name=base_necklace, swing1LimitDegrees=3, swing2LimitDegrees=25, twistLimitDegrees=3}
 {bone1Name=dyn_necklace_01, bone2Name=dyn_necklace_00, swing1LimitDegrees=10, swing2LimitDegrees=25, twistLimitDegrees=5}
 {bone1Name=dyn_l_wing_00, bone2Name=base_l_wing_00, swing1LimitDegrees=5, swing2LimitDegrees=8, twistLimitDegrees=1}
 {bone1Name=dyn_l_wing_01, bone2Name=dyn_l_wing_00, swing1LimitDegrees=5, swing2LimitDegrees=5, twistLimitDegrees=1}
 {bone1Name=dyn_zip, bone2Name=base_zip, swing1LimitDegrees=5, swing2LimitDegrees=45, twistLimitDegrees=1}
 {bone1Name=dyn_l_ear, bone2Name=base_l_ear, swing1LimitDegrees=20, swing2LimitDegrees=0, twistLimitDegrees=0}
 {bone1Name=dyn_r_wing_00, bone2Name=base_r_wing_00, swing1LimitDegrees=5, swing2LimitDegrees=8, twistLimitDegrees=1}
 {
  bone1Name=clavicle_r
  bone2Name=spine_05
  swing1LimitDegrees=45
  swing1Motion=Locked
  swing2LimitDegrees=45
  swing2Motion=Locked
  twistLimitDegrees=45
  twistMotion=Locked
 }
 {
  bone1Name=clavicle_l
  bone2Name=spine_05
  swing1LimitDegrees=45
  swing1Motion=Free
  swing2LimitDegrees=45
  swing2Motion=Free
  twistLimitDegrees=45
  twistMotion=Free
 }
 {bone1Name=dyn_r_wing_01, bone2Name=dyn_r_wing_00, swing1LimitDegrees=5, swing2LimitDegrees=5, twistLimitDegrees=1}
 {bone1Name=upperarm_r, bone2Name=clavicle_r, swing1LimitDegrees=5, swing2LimitDegrees=8, twistLimitDegrees=1}
 {
  bone1Name=upperarm_l
  bone2Name=clavicle_l
  swing1LimitDegrees=45
  swing1Motion=Free
  swing2LimitDegrees=45
  swing2Motion=Free
  twistLimitDegrees=45
  twistMotion=Free
 }
 {bone1Name=dyn_r_ear, bone2Name=base_r_ear, swing1LimitDegrees=20, swing2LimitDegrees=0, twistLimitDegrees=0}
]
