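# Animation Blueprint for the ForestBath backpack cosmetic, keeping it fitted to the wearer's upper body and adding secondary physics.
# Each update caches world transforms of spine_05, both clavicles and both upper arms from the attach-parent body mesh, plus a gravity override.
# The AnimGraph writes those transforms onto matching pack bones in world space, nudges the pack bone, then runs RigidBody simulation in its space using that gravity.
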AnimBlueprint /BRCosmetics/Accessories/FORT_Backpacks/Backpack_F_MED_ForestBath/Meshes/F_MED_ForestBath_Pack_AnimBP
target_skeleton=&/F_MED_ForestBath_Pack_Skeleton
graphs [
 AnimGraph {nodes [
   n1 = "Transform (Modify) Bone - Bone: spine_05"(
    ComponentPose=Local To Component(LocalPose=Input Pose())
    Alpha=1
    Translation=Get spine5_location() -> (spine5_location:Vector)
    Rotation=Get spine5_rotation() -> (spine5_rotation:Rotator)
    Scale="1,1,1"
   ) {
    rotationMode=BMM_Replace
    rotationSpace=BCS_WorldSpace
    scaleSpace=BCS_BoneSpace
    translationMode=BMM_Replace
    translationSpace=BCS_WorldSpace
    alphaBoolBlend {blendOption=Linear}
    alphaScaleBiasClamp {inRange {max=0, min=-40}, outRange {max=0, min=1}}
    boneToModify {boneName=spine_05}
   }
   n2 = "Transform (Modify) Bone - Bone: clavicle_l"(
    ComponentPose=@n1
    Alpha=1
    Translation=Get clavicle_left_location() -> (clavicle_left_location:Vector)
    Rotation=Get clavicle_left_rotation() -> (clavicle_left_rotation:Rotator)
    Scale="1,1,1"
   ) {
    rotationMode=BMM_Replace
    rotationSpace=BCS_WorldSpace
    scaleSpace=BCS_BoneSpace
    translationMode=BMM_Replace
    translationSpace=BCS_WorldSpace
    alphaBoolBlend {blendOption=Linear}
    alphaScaleBiasClamp {inRange {max=0, min=-40}, outRange {max=0, min=1}}
    boneToModify {boneName=clavicle_l}
   }
   n3 = "Transform (Modify) Bone - Bone: clavicle_r"(
    ComponentPose=@n2
    Alpha=1
    Translation=Get clavicle_right_location() -> (clavicle_right_location:Vector)
    Rotation=Get clavicle_right_rotation() -> (clavicle_right_rotation:Rotator)
    Scale="1,1,1"
   ) {
    rotationMode=BMM_Replace
    rotationSpace=BCS_WorldSpace
    scaleSpace=BCS_BoneSpace
    translationMode=BMM_Replace
    translationSpace=BCS_WorldSpace
    alphaBoolBlend {blendOption=Linear}
    alphaScaleBiasClamp {inRange {max=0, min=-40}, outRange {max=0, min=1}}
    boneToModify {boneName=clavicle_r}
   }
   n4 = "Transform (Modify) Bone - Bone: upperarm_l"(
    ComponentPose=@n3
    Alpha=1
    Translation=Get upperarm_left_location() -> (upperarm_left_location:Vector)
    Rotation=Get upperarm_left_rotation() -> (upperarm_left_rotation:Rotator)
    Scale="1,1,1"
   ) {
    rotationMode=BMM_Replace
    rotationSpace=BCS_WorldSpace
    scaleSpace=BCS_BoneSpace
    translationMode=BMM_Replace
    translationSpace=BCS_WorldSpace
    alphaBoolBlend {blendOption=Linear}
    alphaScaleBiasClamp {inRange {max=0, min=-40}, outRange {max=0, min=1}}
    boneToModify {boneName=upperarm_l}
   }
   n5 = "Transform (Modify) Bone - Bone: upperarm_r"(
    ComponentPose=@n4
    Alpha=1
    Translation=Get upperarm_right_location() -> (upperarm_right_location:Vector)
    Rotation=Get upperarm_right_rotation() -> (upperarm_right_rotation:Rotator)
    Scale="1,1,1"
   ) {
    rotationMode=BMM_Replace
    rotationSpace=BCS_WorldSpace
    scaleSpace=BCS_BoneSpace
    translationMode=BMM_Replace
    translationSpace=BCS_WorldSpace
    alphaBoolBlend {blendOption=Linear}
    alphaScaleBiasClamp {inRange {max=0, min=-40}, outRange {max=0, min=1}}
    boneToModify {boneName=upperarm_r}
   }
   n6 = "Transform (Modify) Bone - Bone: pack"(ComponentPose=@n5, Translation="0,1.25,1.25", Rotation="0,0,0") {
    translationMode=BMM_Additive
    translationSpace=BCS_BoneSpace
    alphaBoolBlend {blendOption=Linear}
    boneToModify {boneName=pack}
    translation {y=1.25, z=1.25}
   }
   n7 = RigidBody!(ComponentPose=@n6, Alpha, OverrideWorldGravity:Vector=Get Gravity Override(), ExternalForce) -> (Pose) {
    bOverrideWorldGravity=true
    overridePhysicsAsset=&/F_MED_ForestBath_Pack_PhysicsAsset
    simulationSpace=BaseBoneSpace
    alphaBoolBlend {blendOption=Linear}
    baseBoneRef {boneName=pack}
    componentAppliedLinearAccClamp=<3000 3600 5000>
    componentLinearAccScale=<0.2 0.2 0.25>
    componentLinearVelScale=<0.3 0.36 0.5>
    simSpaceSettings {worldAlpha=0.2}
   }
   Output Pose(Result=Component To Local(ComponentPose=@n7)) {layerGroup=None}
  ]}
 EventGraph {nodes [
   Event BlueprintUpdateAnimation()
   > n1 = Gravity Override State Function DEPRECATED(CharacterPartAnimInstance=Self-Reference(), GameTimeInSeconds=GetGameTimeInSeconds(), GravityOverrideParams=Get Gravity Override Params())
   > Set Gravity Override(Gravity Override=@n1)
   n2 = GetPartSkeletalMeshComponent()
   > Is Valid(InputObject=@n2)
   n3 = GetAttachParent!(self:Scene Component Object Reference=@n2) -> (ReturnValue:Scene Component Object Reference)
   > Is Valid(InputObject=@n3)
   > Set Body_Mesh(Body_Mesh:Scene Component Object Reference=@n3) -> (Output_Get:Scene Component Object Reference)
   > n4 = Sequence!(execute) -> (then_0, then_1, then_2:Exec, then_3:Exec, then_4:Exec, then_5:Exec, then_6:Exec)
   n5 = Get Body_Mesh() -> (Body_Mesh:Scene Component Object Reference)
   n6 = BreakTransform(InTransform=GetSocketTransform(self=@n5, InSocketName=root, TransformSpace=RTS_World))
   Set root_location(execute=@n4.then_6, root_location:Vector=@n6) -> (Output_Get:Vector)
   > Set root_space_rotation(root_space_rotation:Rotator=@n6.Rotation) -> (Output_Get:Rotator)
   n7 = BreakTransform(InTransform=GetSocketTransform(self=@n5, InSocketName=spine_05, TransformSpace=RTS_World))
   Set spine5_location(execute=@n4.then_0, spine5_location:Vector=@n7) -> (Output_Get:Vector)
   > Set spine5_rotation(spine5_rotation:Rotator=@n7.Rotation) -> (Output_Get:Rotator)
   n8 = BreakTransform(InTransform=GetSocketTransform(self=@n5, InSocketName=clavicle_l, TransformSpace=RTS_World))
   Set clavicle_left_location(execute=@n4.then_1, clavicle_left_location:Vector=@n8) -> (Output_Get:Vector)
   > Set clavicle_left_rotation(clavicle_left_rotation:Rotator=@n8.Rotation) -> (Output_Get:Rotator)
   n9 = BreakTransform(InTransform=GetSocketTransform(self=@n5, InSocketName=clavicle_r, TransformSpace=RTS_World))
   Set clavicle_right_location(execute=@n4.then_2, clavicle_right_location:Vector=@n9) -> (Output_Get:Vector)
   > Set clavicle_right_rotation(clavicle_right_rotation:Rotator=@n9.Rotation) -> (Output_Get:Rotator)
   n10 = BreakTransform(InTransform=GetSocketTransform(self=@n5, InSocketName=upperarm_r, TransformSpace=RTS_World))
   Set upperarm_right_location(execute=@n4.then_4, upperarm_right_location:Vector=@n10) -> (Output_Get:Vector)
   > Set upperarm_right_rotation(upperarm_right_rotation:Rotator=@n10.Rotation) -> (Output_Get:Rotator)
   n11 = BreakTransform(InTransform=GetSocketTransform(self=@n5, InSocketName=upperarm_l, TransformSpace=RTS_World))
   Set upperarm_left_location(execute=@n4.then_3, upperarm_left_location:Vector=@n11) -> (Output_Get:Vector)
   > Set upperarm_left_rotation(upperarm_left_rotation:Rotator=@n11.Rotation) -> (Output_Get:Rotator)
   n12 = BreakTransform(InTransform=GetSocketTransform(self=@n5, InSocketName=root, TransformSpace=RTS_World))
   Set body_root_location(execute=@n4.then_5, body_root_location:Vector=@n12) -> (Output_Get:Vector)
   > Set body_root_rotation(body_root_rotation:Rotator=@n12.Rotation) -> (Output_Get:Rotator)
   TryGetPawnOwner()
  ]}
]
variables [
 Body_Mesh {type=SceneComponent, default=null}
 body_root_location {type=Vector}
 body_root_rotation {type=Rotator}
 clavicle_left_location {type=Vector}
 clavicle_left_rotation {type=Rotator}
 clavicle_right_location {type=Vector}
 clavicle_right_rotation {type=Rotator}
 Gravity Override {type=Vector}
 Gravity Override Params {type=GravityOverrideParamsStruct}
 root_location {type=Vector}
 root_space_rotation {type=Rotator}
 spine5_location {type=Vector}
 spine5_rotation {type=Rotator}
 upperarm_left_location {type=Vector}
 upperarm_left_rotation {type=Rotator}
 upperarm_right_location {type=Vector}
 upperarm_right_rotation {type=Rotator}
]
